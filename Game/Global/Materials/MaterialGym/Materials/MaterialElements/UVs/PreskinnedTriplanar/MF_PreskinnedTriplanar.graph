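# Material function for triplanar texture projection in pre-skinned local space, keeping the projection fixed to the mesh under skinning.
# Vertex-interpolated local position is normalized by the largest bounds extent and centered, then split into three planar UV sets with tiling, offset and step-based sign flips.
# Contrast-adjusted absolute X and Z normal components provide blend masks.
# It outputs UVs, masks, swizzle, a three-axis sample and a cheaper two-axis sample.
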MaterialFunction /Game/Global/Materials/MaterialGym/Materials/MaterialElements/UVs/PreskinnedTriplanar/MF_PreskinnedTriplanar
graph {nodes [
  FunctionOutput(None=NamedRerouteUsage()) {description="Sampled Output on all 3 axes ", outputName=Sample XYZ}
  FunctionOutput(None=NamedRerouteUsage()) {description=Projected UV for X Axis, outputName=UV X, sortPriority=10}
  FunctionOutput(None=NamedRerouteUsage()) {description=Projected UV for Y Axis, outputName=UV Y, sortPriority=11}
  FunctionOutput(None=NamedRerouteUsage()) {description=Projected UV for Z Axis, outputName=UV Z, sortPriority=12}
  FunctionOutput(None=NamedRerouteUsage()) {description=Blend mask for UV X, outputName=UV Blend X, sortPriority=20}
  FunctionOutput(None=NamedRerouteUsage()) {description=Blend mask for UV Z, outputName=UV Blend Z, sortPriority=21}
  FunctionOutput(None=NamedRerouteUsage()) {description="Sampled output on 2 axes.  Cheaper than XYZ", outputName=Sample XY, sortPriority=1}
  n1 = Constant3Vector() {constant=<rgba -2 2 2 0>}
  n2 = Add(A=Step(X=NamedRerouteUsage())) {constA=-0.5, constB=-0.5}
  FunctionOutput(None=NamedRerouteDeclaration(None=Multiply(A=@n1, B=@n2)) {name=UVSwizzle, nodeColor=<rgba 0 1 0.0625 1>}) {outputName=UVSwizzle}
  n3 = FunctionInput() {
   bUsePreviewValueAsDefault=true
   description=Blend factor for the projected texture samples.
   inputName=Blend Contrast
   inputType=FunctionInput_Scalar
   sortPriority=4
   previewValue=<1 0 0 1>
  }
  NamedRerouteDeclaration(None=@n3) {name=BlendContrast, nodeColor=<rgba 1 0 0.5625 1>}
  NamedRerouteDeclaration(None=TextureSample(UVs=NamedRerouteUsage(), Tex=NamedRerouteUsage()) {samplerSource=SSM_Wrap_WorldGroupSettings}.RGBA) {name=SampleX, nodeColor=<rgba 1 0.0042 0 1>}
  NamedRerouteDeclaration(None=TextureSample(UVs=NamedRerouteUsage(), Tex=NamedRerouteUsage()) {samplerSource=SSM_Wrap_WorldGroupSettings}.RGBA) {name=SampleY, nodeColor=<rgba 0.0839 1 0 1>}
  NamedRerouteDeclaration(None=TextureSample(UVs=NamedRerouteUsage(), Tex=NamedRerouteUsage()) {samplerSource=SSM_Wrap_WorldGroupSettings}.RGBA) {name=SampleZ, nodeColor=<rgba 0 0.0094 1 1>}
  n4 = NamedRerouteDeclaration(None=VertexInterpolator(VS=PreSkinnedNormal()).PS) {name=NormalLS, nodeColor=<rgba 1 0.7031 0 1>}
  n5 = Abs(None=ComponentMask(None=@n4) {r=true})
  n6 = NamedRerouteUsage()
  NamedRerouteDeclaration(None=ComponentMask(None=MaterialFunctionCall(In=@n5, Contrast=@n6) {materialFunction=&CheapContrast}.Result) {r=true}) {name=BlendX, nodeColor=<rgba 1 0 0.7266 1>}
  n7 = Abs(None=ComponentMask(None=@n4) {b=true})
  NamedRerouteDeclaration(None=ComponentMask(None=MaterialFunctionCall(In=@n7, Contrast=@n6) {materialFunction=&CheapContrast}.Result) {r=true}) {name=BlendZ, nodeColor=<rgba 0 1 0.9531 1>}
  n8 = LinearInterpolate(A=NamedRerouteUsage(), B=NamedRerouteUsage(), Alpha=NamedRerouteUsage()) {desc=Float3}
  NamedRerouteDeclaration(None=LinearInterpolate(A=NamedRerouteDeclaration(None=@n8) {name=SampleXY, nodeColor=<rgba 1 0.9141 0 1>}, B=NamedRerouteUsage(), Alpha=NamedRerouteUsage())) {name=SampleXYZ, nodeColor=<rgba 1 0.6328 0 1>}
  n9 = FunctionInput() {
   bUsePreviewValueAsDefault=true
   description="UV Offset, can be used to warp UVs as well."
   inputName=UV Offset
   inputType=FunctionInput_Vector2
   sortPriority=2
  }
  NamedRerouteDeclaration(None=@n9) {name=UVOffset, nodeColor=<rgba 0.0078 1 0 1>}
  n10 = FunctionInput() {
   bUsePreviewValueAsDefault=true
   description=Scale param for UVs
   inputName=UV Tiling
   inputType=FunctionInput_Vector2
   sortPriority=1
   previewValue=<1 1 0 1>
  }
  n11 = Max(A=Abs(None=@n10)) {constB=0}
  n12 = Sign(None=@n10)
  NamedRerouteDeclaration(None=Divide(B=Multiply(A=@n11, B=@n12))) {name=UVScale, nodeColor=<rgba 0.0078 1 0 1>}
  MaterialExpressionPreSkinnedLocalBounds_1 = PreSkinnedLocalBounds()
  n13 = Subtract(A=VertexInterpolator(VS=LocalPosition() {localOrigin=InstancePreSkinning}.XYZ).PS, B=@MaterialExpressionPreSkinnedLocalBounds_1.Min)
  n14 = Convert(None=@MaterialExpressionPreSkinnedLocalBounds_1.Extents) {
   nodeName=Break Float3
   convertInputs [{
     type=Vector3
     expressionInput {
      expression=&self:MaterialExpressionPreSkinnedLocalBounds_1
      inputName=None
      mask=1
      maskA=0
      maskB=1
      maskG=1
      maskR=1
      outputIndex=1
     }
    }]
   convertOutputs [{type=Scalar}, {type=Scalar}, {type=Scalar}]
  }
  n15 = Max(A=Max(A=@n14, B=@n14), B=@n14)
  n16 = Multiply(
   A=Subtract(A=Divide(A=@n15, B=@MaterialExpressionPreSkinnedLocalBounds_1.Extents))
   B=@MaterialExpressionPreSkinnedLocalBounds_1.Half Extents
  )
  NamedRerouteDeclaration(None=Divide(A=Add(A=@n13, B=@n16), B=@n15)) {name=UV0-1, nodeColor=<rgba 0.6172 0 1 1>}
  n17 = FunctionInput(Preview=TextureObject() {texture=&T_BlacksRGB}) {
   bUsePreviewValueAsDefault=true
   description=Texture to be sampled with triplanar uvs
   inputName=InputTexture
   inputType=FunctionInput_Texture2D
  }
  NamedRerouteDeclaration(None=@n17) {name=TextureDefault, nodeColor=<rgba 0 0.4844 1 1>}
  n18 = MaterialFunctionCall(UVs=ComponentMask(None=NamedRerouteUsage()) {b=true, g=true}, Texture Scale=NamedRerouteUsage()) {materialFunction=&ScaleUVsByCenter}
  n19 = AppendVector(A=ComponentMask(None=NamedRerouteUsage()) {r=true}, B=Constant() {r=-1})
  NamedRerouteDeclaration(None=Add(A=Multiply(A=@n18.UVs, B=@n19), B=NamedRerouteUsage())) {name=UVX, nodeColor=<rgba 1 0 0.0227 1>}
  n20 = MaterialFunctionCall(UVs=ComponentMask(None=NamedRerouteUsage()) {b=true, r=true}, Texture Scale=NamedRerouteUsage()) {materialFunction=&ScaleUVsByCenter}
  n21 = AppendVector(A=ComponentMask(None=NamedRerouteUsage()) {g=true}, B=Constant() {r=-1})
  NamedRerouteDeclaration(None=Add(A=Multiply(A=@n20.UVs, B=@n21), B=NamedRerouteUsage())) {name=UVY, nodeColor=<rgba 0.1448 1 0.0938 1>}
  n22 = MaterialFunctionCall(UVs=ComponentMask(None=NamedRerouteUsage()) {g=true, r=true}, Texture Scale=NamedRerouteUsage()) {materialFunction=&ScaleUVsByCenter}
  n23 = AppendVector(A=Constant() {r=1}, B=ComponentMask(None=NamedRerouteUsage()) {b=true})
  NamedRerouteDeclaration(None=Add(A=Multiply(A=@n22.UVs, B=@n23), B=NamedRerouteUsage())) {name=UVZ, nodeColor=<rgba 0 0.0553 1 1>}
 ]}
