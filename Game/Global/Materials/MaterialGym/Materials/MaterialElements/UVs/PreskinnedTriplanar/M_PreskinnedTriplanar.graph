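# Material Gym example of pre-skinned triplanar mapping, set up as a Default Lit material for skeletal meshes and mesh deformers.
# MF_PreskinnedTriplanar does the projection work, fed by a texture parameter that defaults to the UV Editor color grid.
# Scalar parameters for X and Y tiling and offset are appended into vectors for the function, alongside a blend contrast control.
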Material /Game/Global/Materials/MaterialGym/Materials/MaterialElements/UVs/PreskinnedTriplanar/M_PreskinnedTriplanar
graph {
 properties {bAutomaticallySetUsageInEditor=false, bUsedWithMeshDeformer=true, bUsedWithSkeletalMesh=true, usedShadingModels=Default Lit}
 nodes [MaterialFunctionCall!(
   InputTexture=TextureObjectParameter() {group=PreskinnedTriplanar, parameterName=Texture, sortPriority=0, texture=/UVEditor/Textures/UVEditorColorGrid_Color}
   UV Tiling=AppendVector(
    A=ScalarParameter() {defaultValue=2, group=PreskinnedTriplanar, parameterName=UVTilingX, sortPriority=1}
    B=ScalarParameter() {defaultValue=2, group=PreskinnedTriplanar, parameterName=UVTilingY, sortPriority=2}
   )
   UV Offset=AppendVector(
    A=ScalarParameter() {defaultValue=0.124, group=PreskinnedTriplanar, parameterName=UVOffsetX, sortPriority=3}
    B=ScalarParameter() {defaultValue=0.212, group=PreskinnedTriplanar, parameterName=UVOffsetY, sortPriority=4}
   )
   Blend Contrast=ScalarParameter() {defaultValue=1, group=PreskinnedTriplanar, parameterName=BlendContrast, sortPriority=7}
  ) {materialFunction=&/MF_PreskinnedTriplanar}]
}
