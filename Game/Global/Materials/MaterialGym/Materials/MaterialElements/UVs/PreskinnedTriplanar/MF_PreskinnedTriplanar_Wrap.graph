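# Material function wrapper around MF_PreskinnedTriplanar that supplies ready-made, grouped material parameters for pre-skinned triplanar texture projection.
# A texture object parameter, per-axis tiling and offset scalars paired into vectors, and a blend contrast scalar feed the inner function.
# Named reroutes organize its results, with outputs for three-axis and cheaper two-axis samples, per-axis UVs, X and Z blend masks, and UVSwizzle.
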MaterialFunction /Game/Global/Materials/MaterialGym/Materials/MaterialElements/UVs/PreskinnedTriplanar/MF_PreskinnedTriplanar_Wrap
graph {nodes [
  FunctionOutput(None=NamedRerouteUsage()) {description="Sampled Output on all 3 axes ", outputName=Sample XYZ}
  FunctionOutput(None=NamedRerouteUsage()) {description=Projected UV for X Axis, outputName=UV X, sortPriority=10}
  FunctionOutput(None=NamedRerouteUsage()) {description=Projected UV for Y Axis, outputName=UV Y, sortPriority=11}
  FunctionOutput(None=NamedRerouteUsage()) {description=Projected UV for Z Axis, outputName=UV Z, sortPriority=12}
  FunctionOutput(None=NamedRerouteUsage()) {description=Blend mask for UV X, outputName=UV Blend X, sortPriority=20}
  FunctionOutput(None=NamedRerouteUsage()) {description=Blend mask for UV Z, outputName=UV Blend Z, sortPriority=21}
  FunctionOutput(None=NamedRerouteUsage()) {description="Sampled output on 2 axes.  Cheaper than XYZ", outputName=Sample XY, sortPriority=1}
  n1 = MaterialFunctionCall!(
   InputTexture=TextureObjectParameter() {group=MF_PreskinnedTriplannar, parameterName=PST_TextureInput, texture=&T_BlacksRGB}
   UV Tiling=AppendVector(
    A=ScalarParameter() {defaultValue=2, group=MF_PreskinnedTriplannar, parameterName=PST_UVTilingX}
    B=ScalarParameter() {defaultValue=2, group=MF_PreskinnedTriplannar, parameterName=PST_UVTilingY}
   )
   UV Offset=AppendVector(
    A=ScalarParameter() {group=MF_PreskinnedTriplannar, parameterName=PST_UVOffsetX}
    B=ScalarParameter() {group=MF_PreskinnedTriplannar, parameterName=PST_UVOffsetY}
   )
   Blend Contrast=ScalarParameter() {defaultValue=1, group=MF_PreskinnedTriplannar, parameterName=PST_BlendContrast}
  ) {materialFunction=&/MF_PreskinnedTriplanar}
  FunctionOutput(None=NamedRerouteDeclaration(None=@n1.UVSwizzle) {name=UVSwizzle, nodeColor=<rgba 0 1 0.0625 1>}) {outputName=UVSwizzle}
  NamedRerouteDeclaration(None=@n1.UV Blend X) {name=BlendX, nodeColor=<rgba 1 0 0.7266 1>}
  NamedRerouteDeclaration(None=@n1.UV Blend Z) {name=BlendZ, nodeColor=<rgba 0 1 0.9531 1>}
  NamedRerouteDeclaration(None=@n1.Sample XYZ) {name=SampleXYZ, nodeColor=<rgba 1 0.6328 0 1>}
  NamedRerouteDeclaration(None=@n1.Sample XY) {name=SampleXY, nodeColor=<rgba 1 0.9141 0 1>}
  NamedRerouteDeclaration(None=@n1.UV X) {name=UVX, nodeColor=<rgba 1 0 0.0227 1>}
  NamedRerouteDeclaration(None=@n1.UV Y) {name=UVY, nodeColor=<rgba 0.1448 1 0.0938 1>}
  NamedRerouteDeclaration(None=@n1.UV Z) {name=UVZ, nodeColor=<rgba 0 0.0553 1 1>}
 ]}
