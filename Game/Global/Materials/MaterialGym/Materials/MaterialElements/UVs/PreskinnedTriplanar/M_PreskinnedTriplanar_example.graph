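# Example material that demonstrates preskinned triplanar projection of animated storm noise on skeletal meshes, with Mesh Deformer usage enabled.
# MF_PreskinnedTriplanar_Wrap supplies per-axis UVs, blend weights and a swizzle through named reroutes.
# Storm noise is sampled three times and nested lerps blend named-reroute values.
# The Y noise first passes through a polar UV remap, whose center offset flips horizontally on the sign of a masked green channel.
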Material /Game/Global/Materials/MaterialGym/Materials/MaterialElements/UVs/PreskinnedTriplanar/M_PreskinnedTriplanar_example
graph {
 properties {bAutomaticallySetUsageInEditor=false, bUsedWithMeshDeformer=true, bUsedWithSkeletalMesh=true, usedShadingModels=Default Lit}
 nodes [
  Constant()
  LinearInterpolate(
   A=LinearInterpolate(A=NamedRerouteUsage(), B=NamedRerouteUsage(), Alpha=NamedRerouteUsage())
   B=NamedRerouteUsage()
   Alpha=NamedRerouteUsage()
  )
  n1 = MaterialFunctionCall() {materialFunction=&/MF_PreskinnedTriplanar_Wrap}
  NamedRerouteDeclaration(None=@n1.UV X) {name=UVX, nodeColor=<rgba 0 1 0.4375 1>}
  NamedRerouteDeclaration(None=@n1.UV Y) {name=UVY, nodeColor=<rgba 1 0.1172 0 1>}
  NamedRerouteDeclaration(None=@n1.UV Z) {name=UVZ, nodeColor=<rgba 0 0.9531 1 1>}
  NamedRerouteDeclaration(None=@n1.UV Blend X) {name=UVBlendX, nodeColor=<rgba 1 0.7266 0 1>}
  NamedRerouteDeclaration(None=@n1.UV Blend Z) {name=UVBlendZ, nodeColor=<rgba 0.4062 0 1 1>}
  NamedRerouteDeclaration(None=MaterialFunctionCall!(InputUV=NamedRerouteUsage()) {materialFunction=/Game/Global/Materials/MaterialGym/Materials/MaterialElements/UVs/AnimatedNoise/MF_StormNoise_Wrap}.Result) {name=NoiseX, nodeColor=<rgba 0 1 0.0859 1>}
  n2 = NamedRerouteUsage()
  n3 = Constant() {r=-0.496}
  n4 = Constant() {r=0.936}
  n5 = NamedRerouteUsage()
  n6 = Constant()
  MaterialExpressionLinearInterpolate_3 = LinearInterpolate(
   A=AppendVector(A=Multiply(A=@n3) {constB=-1}, B=@n4)
   B=AppendVector(A=@n3, B=@n4)
   Alpha=If(A=ComponentMask(None=@n5) {g=true}, "A > B"=Constant() {r=1}, "A == B"=@n6, "A < B"=@n6)
  )
  n7 = Convert(None=@MaterialExpressionLinearInterpolate_3) {
   nodeName=Break Float2
   convertInputs [{
     type=Vector2
     expressionInput {
      expression=&self:MaterialExpressionLinearInterpolate_3
      inputName=None
      mask=0
      maskA=0
      maskB=0
      maskG=0
      maskR=0
      outputIndex=0
     }
    }]
   convertOutputs [{type=Scalar}, {type=Scalar}]
  }
  NamedRerouteDeclaration(None=MaterialFunctionCall(InputUV=MaterialFunctionCall!(UVInput=@n2, PolarUV_CenterOffsetR=@n7, PolarUV_CenterOffsetG=@n7) {materialFunction=/Game/Global/Materials/MaterialGym/Materials/MaterialElements/UVs/UVMainpulation/PolarUV/MF_Polar_Wrap}.OutputUV) {materialFunction=&MF_StormNoise_Wrap}.Result) {name=NoiseY, nodeColor=<rgba 1 0 0.2344 1>}
  NamedRerouteDeclaration(None=MaterialFunctionCall(InputUV=NamedRerouteUsage()) {materialFunction=&MF_StormNoise_Wrap}.Result) {name=NoiseZ, nodeColor=<rgba 0 0.5547 1 1>}
  NamedRerouteDeclaration(None=@n1.UVSwizzle) {name=UVSwizzle, nodeColor=<rgba 0.3594 1 0 1>}
 ]
}
